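; Blueprint interface for Creative character override abilities to query and refresh customization data.
; GetCreativeCustomizationData takes a requester object and returns customization data, with placeholder default outputs only.
; RequestUpdateEventData takes a requesting GA Creative Override Customization ability, letting implementers update that ability's event data.
; It declares signatures only; implementing classes supply the logic.
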(fn GetCreativeCustomizationData
  (params (Requester :type Object%20Reference))
  (return 0 false 0))

(fn RequestUpdateEventData
  (params (Requesting%20Ability :type GA%20Creative%20Override%20Customization%20Object%20Reference))
  (return))
